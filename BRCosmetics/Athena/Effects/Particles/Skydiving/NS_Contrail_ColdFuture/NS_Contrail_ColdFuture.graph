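# Skydiving contrail cosmetic that trails ice-themed mesh chains, fog and snowflakes from skeletal mesh bones.
# Guide_Ribbon lays out one ribbon per chain, bent by player velocity, fixed offsets and a dive-scaled sine sway.
# Ring and head meshes follow those ribbons through StickMeshToRibbon attribute readers, while separate emitters spawn rotating trail meshes and sprites.
# ParaGlideLeanAlpha adjusts scale, spawn rates and motion, and time-of-day inputs tune color and alpha.
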NiagaraSystem /BRCosmetics/Athena/Effects/Particles/Skydiving/NS_Contrail_ColdFuture/NS_Contrail_ColdFuture
dependencies {
 usedDataInterfaces [
  &NiagaraDataInterfaceCurve
  &NiagaraDataInterfaceArrayColor
  &NiagaraDataInterfaceParticleRead
  &NiagaraDataInterfaceArrayFloat3
  &NiagaraDataInterfaceArrayFloat
  &NiagaraDataInterfaceVectorCurve
  &NiagaraDataInterfaceColorCurve
  &NiagaraDataInterfaceSkeletalMesh
 ]
 usedDynamicInputs [&SimulationPosition, null]
 usedModules [
  &self:SystemSpawnScript.SystemScriptSource.SystemScriptGraph.NiagaraNodeAssignment_1.SetVariables_#1
  &SystemState
 ]
}
emitters {
 Contrail_Mesh_Heads {
  emitterUpdateScript [
   &EmitterState(
    Inactive Response=::NewEnumerator1
    Life Cycle Mode=::NewEnumerator1
    Loop Behavior=::NewEnumerator1
    Loop Duration Mode=::NewEnumerator1
    UseLoopDelay=false
   )
   &SpawnBurst_Instantaneous(Spawn Count=$System.ChainNumber, Use Spawn Probability=false, UseLoopCountLimit=false)
  ]
  particleSpawnScript [
   &PartitionParticles(Number Of Partitions=$System.ChainNumber, Particles Per Partition=1, Partition Mode:ENiagara_PartitionMode_v2=::NewEnumerator1)
   &InitializeParticle(Color Mode=::NewEnumerator0, Mesh Scale Mode=::NewEnumerator3, Mesh Uniform Scale=0.9, UsePositionOffset=false, Write Material Random=true)
   /SetVariables(Particles.MeshIndex=$Particles.PartitionParticles.Partition) {isSetParameters=true}
  ]
  particleUpdateScript [
   &ParticleState(Kill Particles When Lifetime Has Elapsed=false, Let Infinitely Lived Particles Die When Emitter Deactivates=true)
   &ScaleMeshSize(Scale Factor=&MakeVector(
     X=1
     Y=1
     Z=&FloatFromCurve(
      FloatCurve=DataInterface() {propertyValues {Curve {keys [
          {arriveTangent=15.9072, interpMode=RCIM_Cubic, leaveTangent=15.9072, tangentMode=RCTM_User, time=0, value=0}
          {arriveTangent=1.2036, interpMode=RCIM_Cubic, leaveTangent=1.2036, tangentMode=RCTM_User, time=0.45, value=1}
          <at 1 1>
         ]}}}
      CurveIndex=$Emitter.Age
     )
    ))
   &ScaleMeshSize(Scale Factor=&Lerp_Vector(A=<1 1 1>, B=<1 1 1.2>, Alpha=$User.ParaGlideLeanAlpha)) {module=ScaleMeshSize001}
   &SkeletalMeshLocation(
    Bone / Socket Index=$Particles.PartitionParticles.Partition
    Bone Sampling Mode=::NewEnumerator4
    MaskSpawnGroup=false
    Orientation Sampling=::NewEnumerator1
    Sampled Position Offset=&Lerp_Vector(A=<0 0 0>, B=$Particles.OffsetPosition, Alpha=0)
    Sampled Position Offset Coordinate Space:ENiagaraSkelMeshTransforms=::NewEnumerator2
    Skeletal Mesh=$System.Skeletal Mesh
   )
   &self:StickMeshToRibbon(
    Index:NiagaraInt32=$Particles.PartitionParticles.Index In Partition
    InterpPosition:NiagaraFloat=&FloatFromCurve(
     FloatCurve=DataInterface() {propertyValues {Curve {keys [
         {arriveTangent=15.9072, interpMode=RCIM_Cubic, leaveTangent=15.9072, tangentMode=RCTM_User, time=0, value=0}
         {arriveTangent=1.2036, interpMode=RCIM_Cubic, leaveTangent=1.2036, tangentMode=RCTM_User, time=0.45, value=1}
         <at 1 1>
        ]}}}
     CurveIndex=$Emitter.Age
    )
    LinkOffset=7
    Particle Attribute Reader=DataInterface() {propertyValues {EmitterBinding {bindingMode=Other, emitterName=Guide_Ribbon}}}
    ParticlePerRibbon:NiagaraInt32=$System.ParticlePerRibbon
    Partition:NiagaraInt32=$Particles.PartitionParticles.Partition
    RibbonSpace=1
   ) {module=StickMeshToRibbon}
   &ScaleColor(
    Scale Alpha=&FNTimeOfDayScalar(Night Scale=0.8, Morning Scale=1, Evening Scale=1)
    ScaleA=true
    ScaleRGB=true
   ) {module=ScaleColor002}
  ]
  rendererClasses [&NiagaraMeshRendererProperties]
  renderers [{rendererClass=&NiagaraMeshRendererProperties, rendererIndex=0}]
 }
 Contrail_Mesh_Rings {
  emitterUpdateScript [
   &EmitterState(
    Inactive Response=::NewEnumerator1
    Life Cycle Mode=::NewEnumerator1
    Loop Behavior=::NewEnumerator1
    Loop Duration Mode=::NewEnumerator1
    UseLoopDelay=false
   )
   &SpawnBurst_Instantaneous(
    Spawn Count=&Multiply_Int(A=$System.ChainNumber, B=$System.LinkPerChain)
    Use Spawn Probability=false
    UseLoopCountLimit=false
   )
  ]
  particleSpawnScript [
   &PartitionParticles(
    Number Of Partitions=$System.ChainNumber
    Particles Per Partition:NiagaraInt32=$System.LinkPerChain
    Partition Mode:ENiagara_PartitionMode_v2=::NewEnumerator1
   )
   &InitializeParticle(
    Color Mode=::NewEnumerator0
    Mesh Scale Mode=::NewEnumerator3
    Mesh Uniform Scale=&Add_Float(
     A=0.5
     B=&Multiply_Float(A=0.075, B=&Multiply_Float(A=$Particles.PartitionParticles.Normalized Index In Partition, B=7))
    )
    UsePositionOffset=false
    Write Material Random=true
   )
   /SetVariables(Particles.OffsetPosition:Vector3f=&MakeVector(
     X=0
     Y=0
     Z=&Add_Float(
      A=1
      B=&Multiply_Float(
       A=&Multiply_Float(A=19, B=&MakeFloatFromVector(Vector=$Particles.Scale, Channel=::NewEnumerator0))
       B=&Multiply_Float(A=$Particles.PartitionParticles.Normalized Index In Partition, B=6)
      )
     )
    )) {isSetParameters=true}
  ]
  particleUpdateScript [
   &ParticleState(Kill Particles When Lifetime Has Elapsed=false, Let Infinitely Lived Particles Die When Emitter Deactivates=true)
   &ScaleMeshSize(Scale Factor=&MakeVector(
     X=1
     Y=1
     Z=&FloatFromCurve(
      FloatCurve=DataInterface() {propertyValues {Curve {keys [
          {arriveTangent=15.9072, interpMode=RCIM_Cubic, leaveTangent=15.9072, tangentMode=RCTM_User, time=0, value=0}
          {arriveTangent=1.2036, interpMode=RCIM_Cubic, leaveTangent=1.2036, tangentMode=RCTM_User, time=0.45, value=1}
          <at 1 1>
         ]}}}
      CurveIndex=$Emitter.Age
     )
    ))
   &ScaleMeshSize(Scale Factor=&Lerp_Vector(A=<1 1 1>, B=<1 1 1.2>, Alpha=$User.ParaGlideLeanAlpha)) {module=ScaleMeshSize001}
   &Color(Color=&MaskLinearColor(
     Mask Source=0
     Mask Parameter=?
     Mask=$Particles.PartitionParticles.Partition
     Masked Linear Color=&MultiplyLinearColor(
      Linear Color=&SelectLinearColorFromArray(
       Array Sampling Mode=::NewEnumerator1
       Direct Array Index=$Particles.PartitionParticles.Index In Partition
       Direct Array Mode=::NewEnumerator1
       Color Selection Array=DataInterface() {propertyValues {ColorData [<rgba 0.0999 0.159 0.3663 1>, <rgba 0.552 0.2918 0.6105 1>, <rgba 0.117 0.4179 0.5647 1>]}}
       Random Seed=?
      )
      Scale Factor=&MakeVector4_FromLinearColor(LinearColor=&FN_Select_Color_from_Time_Of_Day(Morning=<rgba 0.7 0.6475 0.595 1>, Day=<rgba 1 1 1 1>, Evening=<rgba 0.8 0.74 0.68 1>, Night=<rgba 0.3 0.3 0.5 1>))
     )
     Passthrough Linear Color=&MultiplyLinearColor(
      Linear Color=&MaskLinearColor(
       Mask Source=1
       Mask Parameter=$Particles.PartitionParticles.Index In Partition
       Mask=6
       Masked Linear Color=<rgba 0.8594 0.1286 0.0065 1>
       Passthrough Linear Color=&SelectLinearColorFromArray(
        Array Sampling Mode=::NewEnumerator1
        Direct Array Index=$Particles.PartitionParticles.Index In Partition
        Direct Array Mode=::NewEnumerator1
        Color Selection Array=DataInterface() {propertyValues {ColorData [<rgba 0.832 0.832 0.832 1>, <rgba 0.2346 0.1328 0.0619 1>]}}
        Random Seed=?
       )
      )
      Scale Factor=&MakeVector4_FromFloat(Float=&FNTimeOfDayScalar(Night Scale=0.8, Morning Scale=1, Evening Scale=1))
     )
    ))
   &SkeletalMeshLocation(
    Bone / Socket Index=$Particles.PartitionParticles.Partition
    Bone Sampling Mode=::NewEnumerator4
    MaskSpawnGroup=false
    Orientation Sampling=::NewEnumerator1
    Sampled Position Offset=&Lerp_Vector(A=<0 0 0>, B=$Particles.OffsetPosition, Alpha=0)
    Sampled Position Offset Coordinate Space:ENiagaraSkelMeshTransforms=::NewEnumerator2
    Skeletal Mesh=$System.Skeletal Mesh
   )
   &self:StickMeshToRibbon(
    ApplyOrientation=true
    Index:NiagaraInt32=$Particles.PartitionParticles.Index In Partition
    InterpPosition:NiagaraFloat=&FloatFromCurve(
     FloatCurve=DataInterface() {propertyValues {Curve {keys [
         {arriveTangent=15.9072, interpMode=RCIM_Cubic, leaveTangent=15.9072, tangentMode=RCTM_User, time=0, value=0}
         {arriveTangent=1.2036, interpMode=RCIM_Cubic, leaveTangent=1.2036, tangentMode=RCTM_User, time=0.45, value=1}
         <at 1 1>
        ]}}}
     CurveIndex=$Emitter.Age
    )
    Particle Attribute Reader=DataInterface() {propertyValues {EmitterBinding {bindingMode=Other, emitterName=Guide_Ribbon}}}
    ParticlePerRibbon:NiagaraInt32=$System.ParticlePerRibbon
    Partition:NiagaraInt32=$Particles.PartitionParticles.Partition
    RibbonSpace=1
   ) {module=StickMeshToRibbon}
  ]
  rendererClasses [&NiagaraMeshRendererProperties]
  renderers [{rendererClass=&NiagaraMeshRendererProperties, rendererIndex=0}]
 }
 Contrail_Mesh_Trail {
  emitterUpdateScript [
   &EmitterState()
   &SpawnRate(
    SpawnRate=&FloatFromCurve(FloatCurve=DataInterface() {propertyValues {Curve {keys [<at 0 10>,<at 1 2>]}}}, CurveIndex=$Engine.System.Age, Scale Curve=1.5)
    Use Spawn Probability=false
   )
  ]
  particleSpawnScript [
   &PartitionParticles()
   &AddVelocity(
    Scale Added Velocity=&Lerp_Vector(A=<1 0 1>, B=<1 1 2>, Alpha=$System.FN_Integrator001.Value)
    Velocity=<0 -600 500>
   ) {enabled=false}
   &InitializeParticle(
    Lifetime Mode=::NewEnumerator1
    Mass Mode=::NewEnumerator1
    Mesh Scale Mode=::NewEnumerator1
    Mesh Uniform Scale Max=1.1
    Sprite Size Mode=::NewEnumerator4
    UsePositionOffset=false
    Write Material Random=true
   ) {module=InitializeParticle001}
   &InitialMeshOrientation(
    Mesh Axis To Orient=<0 1 0>
    Mesh Orientation Mode=::NewEnumerator2
    Orientation Vector=<0 1 0>
    Rotation=&RandomRangeVector(Minimum=<0 0 -360>, Maximum=<0 0 360>)
    Use Rotation Vector=true
   )
   &InitialMeshOrientation(
    Mesh Axis To Orient=<0 1 0>
    Mesh Orientation Mode=::NewEnumerator2
    Orientation Vector=<0 1 0>
    Rotation=&Lerp_Vector(
     A=&MaskVector(Mask Source=0, Mask Parameter=?, Mask=$Particles.PartitionParticles.Partition, Masked Vector=<-0.07 0 0>, Passthrough Vector=<-0.06 0 0>)
     B=&MaskVector(Mask Source=0, Mask Parameter=?, Mask=$Particles.PartitionParticles.Partition, Masked Vector=<-0.07 0 0>, Passthrough Vector=<-0.06 0 0>)
     Alpha=$User.ParaGlideLeanAlpha
    )
    Use Rotation Vector=true
   ) {enabled=false, module=InitialMeshOrientation001}
  ]
  particleUpdateScript [
   &ParticleState()
   &UpdateMeshOrientation(
    Rotation Rate=&SelectFloatFromArray(Direct Array Index=$Particles.PartitionParticles.Partition, Float Selection Array=DataInterface() {propertyValues {FloatData [-1,1]}})
    Rotation Vector=<0 0 1>
   )
   &SkeletalMeshLocation(
    Bone / Socket Index=$Particles.PartitionParticles.Partition
    Bone Sampling Mode=::NewEnumerator4
    Inherit Velocity=false
    Interpolate Mesh Sampling=false
    MaskSpawnGroup=false
    Sampled Position Offset=<0 0 -10>
    Skeletal Mesh=$System.Skeletal Mesh
   )
   &ScaleMeshSize(Scale Factor=&VectorFromCurve(
     VectorCurve=DataInterface() {propertyValues {
       XCurve {keys [<at 0 1>,<at 1 1>]}
       YCurve {keys [<at 0 1>,<at 1 1>]}
       ZCurve {keys [
         {interpMode=RCIM_Cubic, leaveTangent=9.4348, tangentMode=RCTM_Break, time=0, value=0}
         {arriveTangent=2.0101, tangentMode=RCTM_User, time=0.2, value=0.6875}
         <at 1 1.5>
        ]}
      }}
     Scale Curve=&MakeVector(X=1, Y=1, Z=&Lerp_Float(A=0.9, B=1.1, Alpha=$User.ParaGlideLeanAlpha))
    ))
   &SolveForcesAndVelocity(Clamp Velocity=false, Limit Acceleration=false)
   &ScaleColor(
    Scale Alpha=&FloatFromCurve(FloatCurve=DataInterface() {propertyValues {Curve {
        defaultValue=0
        keys [
         {arriveTangent=4.1785, interpMode=RCIM_Cubic, leaveTangent=4.1785, tangentMode=RCTM_User, time=0, value=0}
         {arriveTangent=1.1179, interpMode=RCIM_Cubic, leaveTangent=1.1179, tangentMode=RCTM_User, time=0.15, value=0.8}
         {arriveTangent=-0.2812, interpMode=RCIM_Cubic, leaveTangent=-0.2812, tangentMode=RCTM_User, time=0.5, value=1}
         {arriveTangent=-6.9585, interpMode=RCIM_Cubic, leaveTangent=-6.9585, tangentMode=RCTM_User, time=1, value=0}
        ]
       }}})
    Scale RGB=&MakeVectorFromLinearColorRGB(Color=&ColorFromCurve(ColorCurve=DataInterface() {propertyValues {
        AlphaCurve {keys []}
        BlueCurve {keys [<at 0 0.9047 RCIM_Cubic>,<at 0.2098 2>,<at 0.4593 2>,<at 0.9258 3>]}
        GreenCurve {keys [<at 0 0.7758 RCIM_Cubic>,<at 0.2098 1.4828>,<at 0.4593 1.3444>,<at 0.9258 0.7792>]}
        RedCurve {keys [<at 0 0.6867 RCIM_Cubic>,<at 0.2098 0.7246>,<at 0.4593 0.3576>,<at 0.9258 0.3177>]}
       }}))
    ScaleA=true
    ScaleRGB=true
   )
   &ScaleColor(
    Scale Alpha=&FNTimeOfDayScalar(Night Scale=0.4, Morning Scale=1, Evening Scale=0.75)
    ScaleA=true
    ScaleRGB=true
   ) {module=ScaleColor002}
  ]
  rendererClasses [&NiagaraMeshRendererProperties]
  renderers [{rendererClass=&NiagaraMeshRendererProperties, rendererIndex=0}]
 }
 Contrail_Particles_IceFog {
  emitterUpdateScript [
   &EmitterState()
   &SpawnRate(SpawnRate=&Lerp_Float(A=15, B=30, Alpha=$User.ParaGlideLeanAlpha), Use Spawn Probability=false)
  ]
  particleSpawnScript [
   &PartitionParticles()
   &InitializeParticle(
    Color=<rgba 1 1 1 0.6>
    Lifetime=&Lerp_Float(A=&RandomRangeFloat(Minimum=2, Maximum=3), B=1, Alpha=$User.ParaGlideLeanAlpha)
    Sprite Rotation Mode=::NewEnumerator2
    Sprite Size Mode=::NewEnumerator1
    Sprite UV Mode=::NewEnumerator3
    Uniform Sprite Size Max=60
    Uniform Sprite Size Min=40
    UsePositionOffset=false
    Write Material Random=true
   )
   &CameraOffset(Camera Offset Amount=&RandomRangeFloat(Minimum=0, Maximum=30))
   &ShapeLocation(MaskSpawnGroup=false, Override Seed=false, Sphere Radius=10, UseOwnerScale=true)
   &SkeletalMeshLocation(
    Bone / Socket Index=$Particles.PartitionParticles.Partition
    Bone Sampling Mode=::NewEnumerator4
    Inherit Velocity=false
    Interpolate Mesh Sampling=false
    MaskSpawnGroup=false
    Sampled Position Offset:NiagaraPosition=$Particles.ShapeLocation.ShapePosition
    Skeletal Mesh=$System.Skeletal Mesh
   ) {module=SkeletalMeshLocation001}
  ]
  particleUpdateScript [
   &ParticleState()
   &ScaleColor(
    Scale Alpha=&FloatFromCurve(
     FloatCurve=DataInterface() {propertyValues {Curve {keys [
         <at 0 0 RCIM_Cubic>
         {interpMode=RCIM_Cubic, tangentMode=RCTM_Break, time=0.2, value=1}
         {arriveTangent=-2.3992, tangentMode=RCTM_User, time=1, value=0}
        ]}}}
     Scale Curve=0.75
    )
    Scale RGB=&MakeVectorFromLinearColorRGB(Color=&ColorFromCurve(ColorCurve=DataInterface() {propertyValues {
        AlphaCurve {keys []}
        BlueCurve {keys [<at 0 0.9047 RCIM_Cubic>,<at 0.1361 0.8963>,<at 0.5172 0.9473>,<at 1 0.9734>]}
        GreenCurve {keys [<at 0 0.7758 RCIM_Cubic>,<at 0.1361 0.7345>,<at 0.5172 0.7023>,<at 1 0.5728>]}
        RedCurve {keys [<at 0 0.6867 RCIM_Cubic>,<at 0.1361 0.491>,<at 0.5172 0.3432>,<at 1 0.2906>]}
       }}))
    ScaleA=true
    ScaleRGB=true
   )
   &ScaleSpriteSize(Uniform Curve Sprite Scale=DataInterface() {propertyValues {Curve {keys [
        {arriveTangent=3.5003, interpMode=RCIM_Cubic, leaveTangent=3.5003, tangentMode=RCTM_User, time=0, value=0.5}
        {arriveTangent=0.5833, interpMode=RCIM_Cubic, leaveTangent=0.5833, time=0.18, value=0.9}
        <at 1.2 1.2 RCIM_Cubic>
       ]}}})
   &GravityForce(Gravity=&MakeVector(
     X=0
     Y=0
     Z=&Lerp_Float(
      A=&RandomRangeFloat(Minimum=35, Maximum=50)
      B=&RandomRangeFloat(Minimum=300, Maximum=500)
      Alpha=$User.ParaGlideLeanAlpha
     )
    ))
   &DynamicMaterialParameters(
    Index 0 Param 1=&FloatFromCurve(FloatCurve=DataInterface() {propertyValues {Curve {keys [<at 0 0 RCIM_Cubic>,<at 1.0011 0.2>]}}})
    Index 1 Param 1=&FloatFromCurve(
     FloatCurve=DataInterface() {propertyValues {Curve {keys [<at 0 0>,<at 0.25 1>]}}}
     Scale Curve=&RandomRangeFloat(Minimum=2, Maximum=1)
    )
    Index 1 Param 2=&FloatFromCurve(
     FloatCurve=DataInterface() {propertyValues {Curve {keys [
         {interpMode=RCIM_Cubic, leaveTangent=3.7422, tangentMode=RCTM_Break, time=0, value=0.2}
         {interpMode=RCIM_Cubic, leaveTangent=0.9066, time=1, value=2}
        ]}}}
     Scale Curve=&RandomRangeFloat(Minimum=0.5, Maximum=1)
    )
    Index 1 Param 3=&FloatFromCurve(FloatCurve=DataInterface() {propertyValues {Curve {keys [
         {interpMode=RCIM_Cubic, leaveTangent=2.8445, tangentMode=RCTM_Break, time=0, value=0.2}
         {interpMode=RCIM_Cubic, leaveTangent=63.2978, time=0.25, value=1}
        ]}}})
    Index 1 Param 4=&FloatFromCurve(
     FloatCurve=DataInterface() {propertyValues {Curve {keys [<at 0 1>,<at 1 0.8>]}}}
     Scale Curve=&RandomRangeFloat(Minimum=0.5, Maximum=1)
    )
    Index 2 Param 1=&FloatFromCurve(FloatCurve=DataInterface() {propertyValues {Curve {keys [{interpMode=RCIM_Cubic, leaveTangent=3.2309, tangentMode=RCTM_Break, time=0.1, value=0}, <at 1 1>]}}})
    Index 2 Param 2=$Particles.MaterialRandom
    Index 2 Param 3=&FloatFromCurve(FloatCurve=DataInterface() {propertyValues {Curve {keys [
         {interpMode=RCIM_Cubic, tangentMode=RCTM_Break, time=0, value=1}
         {interpMode=RCIM_Cubic, leaveTangent=11.1233, tangentMode=RCTM_Break, time=1, value=2}
        ]}}})
    Index 2 Param 4=0.75
    Module.Index0Param0Write=true
    Module.Index0Param1Write=true
    Module.Index0Param2Write=true
    Module.Index0Param3Write=true
    Module.Index1Param0Write=true
    Module.Index1Param1Write=true
    Module.Index1Param2Write=true
    Module.Index1Param3Write=true
    Module.Index2Param0Write=true
    Module.Index2Param1Write=true
    Module.Index2Param2Write=true
    Module.Index2Param3Write=true
    Write Parameter Index 1=true
    Write Parameter Index 2=true
   )
   &ScaleColor(
    Scale Alpha=&FNTimeOfDayScalar(Night Scale=0.6, Morning Scale=1, Evening Scale=0.8)
    ScaleA=true
    ScaleRGB=true
   ) {module=ScaleColor002}
   &SolveForcesAndVelocity(Clamp Velocity=false, Limit Acceleration=false)
  ]
  rendererClasses [&NiagaraSpriteRendererProperties]
  renderers [{rendererClass=&NiagaraSpriteRendererProperties}]
 }
 Contrail_Particles_Snowflakes {
  emitterUpdateScript [
   &EmitterState()
   &SpawnRate(SpawnRate=&Lerp_Float(A=10, B=20, Alpha=$User.ParaGlideLeanAlpha), Use Spawn Probability=false)
  ]
  particleSpawnScript [
   &PartitionParticles()
   &InitializeParticle(
    Lifetime=&Lerp_Float(
     A=&RandomRangeFloat(Minimum=1, Maximum=2)
     B=&RandomRangeFloat(Minimum=0.7, Maximum=1.2)
     Alpha=$User.ParaGlideLeanAlpha
    )
    Mass Mode=::NewEnumerator3
    Sprite Rotation Mode=::NewEnumerator2
    Sprite Size Mode=::NewEnumerator1
    Uniform Sprite Size Max=15
    Uniform Sprite Size Min=3
    UsePositionOffset=false
    Write Material Random=true
   ) {module=InitializeParticle001}
   &ShapeLocation(MaskSpawnGroup=false, Override Seed=false, Sphere Radius=10, UseOwnerScale=true)
   &SkeletalMeshLocation(
    Bone / Socket Index=$Particles.PartitionParticles.Partition
    Bone Sampling Mode=::NewEnumerator4
    Inherit Velocity=false
    Interpolate Mesh Sampling=false
    MaskSpawnGroup=false
    Sampled Position Offset:NiagaraPosition=$Particles.ShapeLocation.ShapePosition
    Skeletal Mesh=$System.Skeletal Mesh
   )
   &AddVelocity(Velocity=<0 0 1>, Velocity Speed Scale=&Lerp_Float(A=1, B=100, Alpha=$User.ParaGlideLeanAlpha))
  ]
  particleUpdateScript [
   &ParticleState()
   &SpriteRotationRate(Rotation Rate=&Lerp_Float(
     A=&RandomRangeFloat(Minimum=40, Maximum=70)
     B=&RandomRangeFloat(Minimum=100, Maximum=130)
     Alpha=$User.ParaGlideLeanAlpha
    ))
   &GravityForce(Gravity=<0 0 100>)
   &ScaleSpriteSize(Uniform Curve Sprite Scale=DataInterface() {propertyValues {Curve {
       defaultValue=0
       keys [
        {arriveTangent=4.1785, interpMode=RCIM_Cubic, leaveTangent=4.1785, tangentMode=RCTM_User, time=0, value=0}
        {arriveTangent=1.1179, interpMode=RCIM_Cubic, leaveTangent=1.1179, tangentMode=RCTM_User, time=0.2042, value=0.7914}
        {arriveTangent=0.0224, interpMode=RCIM_Cubic, leaveTangent=0.0224, tangentMode=RCTM_User, time=1, value=1}
       ]
      }}})
   &ScaleColor(
    Scale Alpha=&FloatFromCurve(FloatCurve=DataInterface() {propertyValues {Curve {
        defaultValue=0
        keys [
         {arriveTangent=4.6614, interpMode=RCIM_Cubic, leaveTangent=4.6614, tangentMode=RCTM_User, time=0, value=0}
         {interpMode=RCIM_Cubic, tangentMode=RCTM_User, time=0.5, value=1}
         {tangentMode=RCTM_User, time=1, value=0}
        ]
       }}})
    Scale RGB=&MakeVectorFromLinearColorRGB(Color=&ColorFromCurve(ColorCurve=DataInterface() {propertyValues {
        AlphaCurve {keys []}
        BlueCurve {keys [<at 0 0.9047 RCIM_Cubic>,<at 0.1361 0.8963>,<at 0.5172 0.9473>,<at 1 0.9734>]}
        GreenCurve {keys [<at 0 0.7758 RCIM_Cubic>,<at 0.1361 0.7345>,<at 0.5172 0.7023>,<at 1 0.5728>]}
        RedCurve {keys [<at 0 0.6867 RCIM_Cubic>,<at 0.1361 0.491>,<at 0.5172 0.3432>,<at 1 0.2906>]}
       }}))
    ScaleA=true
    ScaleRGB=true
   )
   &SolveForcesAndVelocity(Clamp Velocity=false, Limit Acceleration=false)
   &ScaleColor(
    Scale Alpha=&FNTimeOfDayScalar(Night Scale=0.1, Morning Scale=0.8, Evening Scale=0.5)
    ScaleA=true
    ScaleRGB=true
   ) {module=ScaleColor002}
  ]
  rendererClasses [&NiagaraSpriteRendererProperties]
  renderers [{rendererClass=&NiagaraSpriteRendererProperties}]
 }
 Guide_Ribbon {
  emitterUpdateScript [
   &EmitterState(
    Inactive Response=::NewEnumerator1
    Life Cycle Mode=::NewEnumerator1
    Loop Behavior=::NewEnumerator1
    Loop Duration Mode=::NewEnumerator1
    UseLoopDelay=false
   )
   &SpawnBurst_Instantaneous(
    Spawn Count=&Multiply_Int(A=$System.ChainNumber, B=$System.ParticlePerRibbon)
    Use Spawn Probability=false
    UseLoopCountLimit=false
   )
  ]
  particleSpawnScript [
   &InitializeParticle(
    Lifetime=5
    Mesh Scale Mode=::NewEnumerator4
    Ribbon Width=2.25
    Ribbon Width Mode=::NewEnumerator1
    Sprite Size Mode=::NewEnumerator3
    UsePositionOffset=false
    Write Material Random=true
   )
   &PartitionParticles(
    Number Of Partitions=$System.ChainNumber
    Particles Per Partition:NiagaraInt32=$System.ParticlePerRibbon
    Partition Mode:ENiagara_PartitionMode_v2=::NewEnumerator1
   )
   /SetVariables(Particles.NormalizedChainPosition:NiagaraFloat=$Particles.PartitionParticles.Normalized Index In Partition) {isSetParameters=true}
   /SetVariables() {isSetParameters=true}
  ]
  particleUpdateScript [
   &ParticleState(Kill Particles When Lifetime Has Elapsed=false)
   &SkeletalMeshLocation(
    Bone / Socket Index=$Particles.PartitionParticles.Partition
    Bone Sampling Mode=::NewEnumerator4
    MaskSpawnGroup=false
    Orientation Sampling=::NewEnumerator1
    Sampled Position Offset=<0 0 -5>
    Sampled Position Offset Coordinate Space:ENiagaraSkelMeshTransforms=::NewEnumerator2
    Skeletal Mesh=$System.Skeletal Mesh
   ) {module=SkeletalMeshLocation002}
   /SetVariables(Particles.Position=&AddVectorToPosition(
     Position=$Particles.Position
     Vector=&Multiply_VectorByFloat(
      Vector=&Add_Vector(
       A=&Multiply_Vector(
        A=&NormalizeVector(
         Vector To Normalize=&FN_FrontEndVector3Modifier(Gameplay=$System.PlayerVelDir, Front End=$System.FrontendVelocity)
         Safe Normalize=true
         Safe Fallback Threshold=0
         Safe Fallback Vector=<0 0 0.001>
        )
        B=&SelectVectorFromArray(
         Direct Array Index=$Particles.PartitionParticles.Partition
         Vector Selection Array=DataInterface() {propertyValues {FloatData [<-0.5 -0.5 -1>, <-0.5 -0.5 -1>, <0.1 -1.5 -1>, <0.1 -0.5 -1>, <-0.5 -1.5 -1>, <-0.5 -1 -1>, <0.1 -1 -1>, <0.1 -0.5 -1>]}}
        )
       )
       B=&Multiply_VectorByFloat(
        Vector=&TransformVector(
         Vector=&SelectVectorFromArray(
          Direct Array Index=$Particles.PartitionParticles.Partition
          Direct Array Mode=::NewEnumerator1
          Vector Selection Array=DataInterface() {propertyValues {FloatData [<3.5 -2 0>, <-1 0 0>, <1 2 0>, <-0.5 0 0>, <0.5 0 0>, <-1 2 0>, <1 1 0>, <-2 1 0>]}}
         )
         Destination Space=::Simulation
        )
        Float=&SelectFloatFromArray(
         Direct Array Index=&SpawnGroup()
         Float Selection Array=DataInterface() {propertyValues {FloatData [0.1,0.15,0.08,0.16,0.15,0.18]}}
        )
       )
      )
      Float=&FloatFromCurve(
       FloatCurve=DataInterface() {propertyValues {Curve {keys [{tangentMode=RCTM_User, time=0, value=0}, {tangentMode=RCTM_User, time=1, value=1}]}}}
       CurveIndex=$Particles.PartitionParticles.Normalized Index In Partition
       Scale Curve=&SelectFloatFromArray(
        Direct Array Index=&SpawnGroup()
        Float Selection Array=DataInterface() {propertyValues {FloatData [97,97,120,100,75,55,94,80]}}
       )
      )
     )
    )) {isSetParameters=true}
   /SetVariables(Particles.Position=&AddVectorToPosition(
     Position=$Particles.Position
     Vector=&Multiply_VectorByFloat(
      Vector=/Niagara/DynamicInputs/Vectors/RotateVector(
       Vector To Rotate:Vector3f=<10 0 3>
       Yaw:NiagaraFloat=&Add_Float(
        A=$Particles.Age
        B=&Multiply_Float(A=&MakeFloatFromInt(Int=&SpawnGroup()), B=2)
       )
       Pitch:NiagaraFloat=0
       Roll:NiagaraFloat=0
       Rotation Rate:NiagaraFloat=&SelectFloatFromArray(
        Direct Array Index=$Particles.PartitionParticles.Partition
        Float Selection Array=DataInterface() {propertyValues {FloatData [100,100,130,120,160,150,120,140]}}
       )
      )
      Float=&Multiply_Float(
       A=&Sine(
        Normalized Angle=&Add_Float(
         A=&Multiply_Float(
          A=$Emitter.Age
          B=&SelectFloatFromArray(
           Direct Array Index=&SpawnGroup()
           Float Selection Array=DataInterface() {propertyValues {FloatData [-2,-2.5,-2.5,-3,-3,-2,-2.5,-2]}}
          )
         )
         B=$Particles.PartitionParticles.Normalized Index In Partition
        )
        Period=0.9
        Scale=&Lerp_Float(A=0.8, B=1.4, Alpha=$System.FN_Contrail_Setup.Dive Value)
        Bias=0
       )
       B=&FloatFromCurve(
        FloatCurve=DataInterface() {propertyValues {Curve {keys [
            {arriveTangent=1.1676, interpMode=RCIM_Cubic, leaveTangent=1.1676, tangentMode=RCTM_User, time=0, value=0}
            {arriveTangent=1.4544, interpMode=RCIM_Cubic, leaveTangent=1.4544, tangentMode=RCTM_User, time=1, value=1}
           ]}}}
        CurveIndex=$Particles.PartitionParticles.Normalized Index In Partition
       )
      )
     )
    )) {isSetParameters=true}
   /SetVariables(Particles.Position=&AddVectorToPosition(
     Position=$Particles.Position
     Vector=&Multiply_VectorByFloat(
      Vector=&SelectVectorFromArray(
       Direct Array Index=$Particles.PartitionParticles.Partition
       Direct Array Mode=::NewEnumerator1
       Vector Selection Array=DataInterface() {propertyValues {FloatData [<-20 10 60>, <0 0 60>]}}
      )
      Float=&FloatFromCurve(
       FloatCurve=DataInterface() {propertyValues {Curve {
          defaultValue=0
          keys [
           {tangentMode=RCTM_User, time=0, value=0}
           <at 0.4979 0.4979>
           <at 0.5587 0.494>
           <at 0.6444 0.6148>
           <at 0.7804 0.6927>
           <at 0.9204 0.8352>
           {tangentMode=RCTM_User, time=1, value=1}
          ]
         }}}
       CurveIndex=$Particles.Initial.PartitionParticles.Normalized Index In Partition
       Scale Curve=&Add_Float(
        A=1
        B=&Multiply_Float(
         A=$Particles.PartitionParticles.Normalized Index In Partition
         B=&Lerp_Float(A=0.8, B=1, Alpha=$User.ParaGlideLeanAlpha)
        )
       )
      )
     )
    )) {isSetParameters=true}
  ]
  rendererClasses [&NiagaraRibbonRendererProperties]
  renderers [{rendererClass=&NiagaraRibbonRendererProperties}]
 }
}
summary {
 systemName=NS_Contrail_ColdFuture
 emitters [
  {emitterName=Contrail_Mesh_Rings, rendererClasses [&NiagaraMeshRendererProperties]}
  {emitterName=Guide_Ribbon, rendererClasses [&NiagaraRibbonRendererProperties]}
  {emitterName=Contrail_Mesh_Heads, rendererClasses [&NiagaraMeshRendererProperties]}
  {emitterName=Contrail_Mesh_Trail, rendererClasses [&NiagaraMeshRendererProperties]}
  {emitterName=Contrail_Particles_Snowflakes, rendererClasses [&NiagaraSpriteRendererProperties]}
  {emitterName=Contrail_Particles_IceFog, rendererClasses [&NiagaraSpriteRendererProperties]}
 ]
 userVariables [
  User.bIsFrontEnd {type=NiagaraBool}
  User.bIsFrontEndPreview {type=NiagaraBool}
  User.ParaGlideLeanAlpha {type=NiagaraFloat, defaultValue=0, description=""}
 ]
}
